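# Level Sequence for the PistolCanary single-tentacle Section01 MaskMakerHit1000 animation, staging the tentacle creature with the MaskMaker character.
# Several tentacle copies (layout, idle, WIP, bake) carry Control Rig tracks, some alongside skeletal animation.
# MaskMaker chains animation clips along a keyed path, growing in steps, with keyed visibility.
# A static CineCamera matches the last of Camera_position's frame-by-frame viewpoints; the camera cut ends well before MaskMaker's action.
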LevelSequence /PistolCanary/Anim/SingleTentacle/Section01/PistolCanary_Tentacle_MaskMakerHit1000_anm
bindings [
 CineCameraActor {
  id=#1
  possessed_class=CineCameraActor
  tracks [{
    type=MovieScene3DTransformTrack
    sections [{
      type=MovieScene3DTransformSection
      channels [
       Location.X {type=MovieSceneScriptingDoubleChannel, default=46662, const [17987.4551,1]}
       Location.Y {type=MovieSceneScriptingDoubleChannel, default=70875.3796, const [84841.2344,1]}
       Location.Z {type=MovieSceneScriptingDoubleChannel, default=8563.014, const [3176.7524,1]}
       Rotation.X {type=MovieSceneScriptingDoubleChannel, default=0, const [-1.1523,1]}
       Rotation.Y {type=MovieSceneScriptingDoubleChannel, default=3.6, const [23.0969,1]}
       Rotation.Z {type=MovieSceneScriptingDoubleChannel, default=-179.2, const [-104.3037,1]}
       Scale.X {type=MovieSceneScriptingDoubleChannel, default=1}
       Scale.Y {type=MovieSceneScriptingDoubleChannel, default=1}
       Scale.Z {type=MovieSceneScriptingDoubleChannel, default=1}
      ]
     }]
   }]
 }
 Tentacle_PistolCanary_Creature_Bake_03 {
  id=#2
  possessed_class=FortSkeletalMeshActor
  tracks [
   {type=MovieSceneSpawnTrack, sections [{type=MovieSceneSpawnSection, channels [None {type=MovieSceneScriptingBoolChannel, default=true}]}]}
   {
    type=MovieScene3DTransformTrack
    sections [{
      type=MovieScene3DTransformSection
      channels [
       Location.X {type=MovieSceneScriptingDoubleChannel, default=36549.9999, const [14287.208,1]}
       Location.Y {type=MovieSceneScriptingDoubleChannel, default=0, const [70618.9219,1]}
       Location.Z {type=MovieSceneScriptingDoubleChannel, default=0, const [3947.8015,1]}
       Rotation.X {type=MovieSceneScriptingDoubleChannel, default=0, const [0,1]}
       Rotation.Y {type=MovieSceneScriptingDoubleChannel, default=0, const [0,1]}
       Rotation.Z {type=MovieSceneScriptingDoubleChannel, default=0, const [0,1]}
       Scale.X {type=MovieSceneScriptingDoubleChannel, default=1, const [1,1]}
       Scale.Y {type=MovieSceneScriptingDoubleChannel, default=1, const [1,1]}
       Scale.Z {type=MovieSceneScriptingDoubleChannel, default=1, const [1,1]}
      ]
     }]
   }
   {
    type=MovieSceneControlRigParameterTrack
    display_name=Tentacle_PistolCanary_Creature_CtrlRig
    sections [{
      type=MovieSceneControlRigParameterSection
      channels [
       control rig options.Location.X {type=MovieSceneScriptingFloatChannel, default=0, const [0,464]}
       control rig options.Location.Y {type=MovieSceneScriptingFloatChannel, default=0, const [0,464]}
       control rig options.Location.Z {type=MovieSceneScriptingFloatChannel, default=0, const [0,464]}
       control rig options.Rotation.X {type=MovieSceneScriptingFloatChannel, default=0, const [0,464]}
       control rig options.Rotation.Y {type=MovieSceneScriptingFloatChannel, default=0, const [0,464]}
       control rig options.Rotation.Z {type=MovieSceneScriptingFloatChannel, default=0, const [0,464]}
       control rig options.Scale.X {type=MovieSceneScriptingFloatChannel, default=1, const [1,464]}
       control rig options.Scale.Y {type=MovieSceneScriptingFloatChannel, default=1, const [1,464]}
       control rig options.Scale.Z {type=MovieSceneScriptingFloatChannel, default=1, const [1,464]}
       Weight {type=MovieSceneScriptingFloatChannel, default=1}
      ]
     }]
   }
  ]
 }
 Tentacle_PistolCanary_Creature {
  id=#3
  possessed_class=FortSkeletalMeshActor
  tracks [
   {type=MovieSceneSpawnTrack, sections [{type=MovieSceneSpawnSection, channels [None {type=MovieSceneScriptingBoolChannel, default=true}]}]}
   {type=MovieSceneSkeletalAnimationTrack, sections []}
   {
    type=MovieScene3DTransformTrack
    sections [{
      type=MovieScene3DTransformSection
      channels [
       Location.X {type=MovieSceneScriptingDoubleChannel, default=36549.9999, const [14287.208,1]}
       Location.Y {type=MovieSceneScriptingDoubleChannel, default=0, const [70618.9219,1]}
       Location.Z {type=MovieSceneScriptingDoubleChannel, default=0, const [3947.8015,1]}
       Rotation.X {type=MovieSceneScriptingDoubleChannel, default=0, const [0,1]}
       Rotation.Y {type=MovieSceneScriptingDoubleChannel, default=0, const [0,1]}
       Rotation.Z {type=MovieSceneScriptingDoubleChannel, default=0, const [0,1]}
       Scale.X {type=MovieSceneScriptingDoubleChannel, default=1, const [1,1]}
       Scale.Y {type=MovieSceneScriptingDoubleChannel, default=1, const [1,1]}
       Scale.Z {type=MovieSceneScriptingDoubleChannel, default=1, const [1,1]}
      ]
     }]
   }
   {
    type=MovieSceneControlRigParameterTrack
    display_name=Tentacle_PistolCanary_Creature_CtrlRig
    sections [{
      type=MovieSceneControlRigParameterSection
      channels [
       control rig options.Location.X {type=MovieSceneScriptingFloatChannel, default=0, const [0,464]}
       control rig options.Location.Y {type=MovieSceneScriptingFloatChannel, default=0, const [0,464]}
       control rig options.Location.Z {type=MovieSceneScriptingFloatChannel, default=0, const [0,464]}
       control rig options.Rotation.X {type=MovieSceneScriptingFloatChannel, default=0, const [0,464]}
       control rig options.Rotation.Y {type=MovieSceneScriptingFloatChannel, default=0, const [0,464]}
       control rig options.Rotation.Z {type=MovieSceneScriptingFloatChannel, default=0, const [0,464]}
       control rig options.Scale.X {type=MovieSceneScriptingFloatChannel, default=1, const [1,464]}
       control rig options.Scale.Y {type=MovieSceneScriptingFloatChannel, default=1, const [1,464]}
       control rig options.Scale.Z {type=MovieSceneScriptingFloatChannel, default=1, const [1,464]}
       Weight {type=MovieSceneScriptingFloatChannel, default=1}
      ]
     }]
   }
  ]
 }
 Camera_position {
  id=#4
  possessed_class=FortStaticMeshActor
  tracks [
   {type=MovieSceneSpawnTrack, sections [{type=MovieSceneSpawnSection, channels [None {type=MovieSceneScriptingBoolChannel, default=true}]}]}
   {
    type=MovieScene3DTransformTrack
    sections [{
      type=MovieScene3DTransformSection
      channels [
       Location.X {
        type=MovieSceneScriptingDoubleChannel
        default=0
        keys [[3690,13994.1279],[3691,46662],[3692,-6830.5537],[3693,13034.7173],[3694,-637.5554],[3695,36329.5273],[3696,12424.2852],[3697,17987.4551]]
       }
       Location.Y {
        type=MovieSceneScriptingDoubleChannel
        default=0
        keys [[3690,99632.6797],[3691,70875.3796],[3692,87218.6875],[3693,82125.2575],[3694,49357.2188],[3695,58227.8281],[3696,84889.1797],[3697,84841.2344]]
       }
       Location.Z {
        type=MovieSceneScriptingDoubleChannel
        default=0
        keys [[3690,7065.0771],[3691,8563.014],[3692,9013.1514],[3693,3129.8077],[3694,9982.4424],[3695,9010.2363],[3696,3129.8076],[3697,3176.7524]]
       }
       Rotation.X {
        type=MovieSceneScriptingDoubleChannel
        default=0
        keys [[3690,0.1003],[3691,0],[3692,0],[3693,-1.1835],[3694,0],[3695,0],[3696,-1.1835],[3697,-1.1523]]
       }
       Rotation.Y {
        type=MovieSceneScriptingDoubleChannel
        default=0
        keys [[3690,6.1663],[3691,3.6],[3692,1.4],[3693,26.419],[3694,-0.4],[3695,1.4],[3696,26.419],[3697,23.0969]]
       }
       Rotation.Z {
        type=MovieSceneScriptingDoubleChannel
        default=0
        keys [[3690,-88.7934],[3691,-179.2],[3692,-37.4],[3693,-83.7285],[3694,54.8],[3695,150.6],[3696,-83.7285],[3697,-104.3037]]
       }
       Scale.X {type=MovieSceneScriptingDoubleChannel, default=1}
       Scale.Y {type=MovieSceneScriptingDoubleChannel, default=1}
       Scale.Z {type=MovieSceneScriptingDoubleChannel, default=1}
      ]
     }]
   }
  ]
 }
 MaskMaker {
  id=#5
  possessed_class=FortPlayerMannequin
  tracks [
   {
    type=MovieSceneSkeletalAnimationTrack
    sections [
     {type=MovieSceneSkeletalAnimationSection, end_frame=2496, start_frame=1800}
     {type=MovieSceneSkeletalAnimationSection, end_frame=2582, start_frame=2475}
     {type=MovieSceneSkeletalAnimationSection, end_frame=2644, start_frame=2576}
     {type=MovieSceneSkeletalAnimationSection, end_frame=2663, start_frame=2634}
     {type=MovieSceneSkeletalAnimationSection, end_frame=2719, start_frame=2657}
     {type=MovieSceneSkeletalAnimationSection, end_frame=2743, start_frame=2714}
     {type=MovieSceneSkeletalAnimationSection, end_frame=2798, start_frame=2737}
     {type=MovieSceneSkeletalAnimationSection, end_frame=3076, start_frame=2779}
     {type=MovieSceneSkeletalAnimationSection, end_frame=3095, start_frame=3075}
     {type=MovieSceneSkeletalAnimationSection, end_frame=3200, start_frame=3091}
     {type=MovieSceneSkeletalAnimationSection, end_frame=3977, start_frame=3182}
    ]
   }
   {type=MovieSceneSpawnTrack, sections [{type=MovieSceneSpawnSection, channels [None {type=MovieSceneScriptingBoolChannel, default=true}]}]}
   {
    type=MovieScene3DTransformTrack
    sections [
     {
      type=MovieScene3DTransformSection
      channels [
       Location.X {type=MovieSceneScriptingDoubleChannel, default=-1950, keys [[1864.5,-100],[3895,-100],[3908,-1250],[3909,-1950],[3954.5,-46555]]}
       Location.Y {
        type=MovieSceneScriptingDoubleChannel
        default=500
        thinned_from=15
        keys [[1881.5,-6320.5469],[1912.5,-6216.229],[2088.5,-3553.2683],[2344,-375.6404],[3086,-375.6404],[3099,-100],[3173.5,-100],[3193.5,-375.6404],[3329.5,115.8782],[3688,115.8782],[3797,-200],[3883.5,-500],[3907,-500],[3954.5,8164]]
       }
       Location.Z {
        type=MovieSceneScriptingDoubleChannel
        default=1100
        thinned_from=15
        keys [[1864.5,1100],[2161.5,1324],[2368.5,2748],[3193.5,2748],[3329.5,3654.0254],[3419,3877.6357],[3735.5,3877.6357],[3797,3800],[3909,3800],[3927.5,13676],[3941,15257],[3954.5,13706]]
       }
       Rotation.X {
        type=MovieSceneScriptingDoubleChannel
        default=20
        thinned_from=18
        keys [[1864.5,0],[2856.5,0],[3075.5,20],[3085.5,46],[3200.5,46],[3257.5,19],[3688.5,19],[3710.5,-5],[3735.5,-10],[3895,-10],[3907,20],[3919.495,120],[3954.5,120]]
       }
       Rotation.Y {type=MovieSceneScriptingDoubleChannel, default=0, const [0,16]}
       Rotation.Z {
        type=MovieSceneScriptingDoubleChannel
        default=100
        thinned_from=15
        keys [[1864.5,0],[2257,0],[2337,180],[2999.0038,200],[3899,200],[3907,100],[3919.495,41],[3954.5,41]]
       }
       Scale.X {
        type=MovieSceneScriptingDoubleChannel
        default=2
        thinned_from=12
        keys [[1864.5,1.25],[2529.0025,1.25],[2558.5025,1.75],[2598.0013,1.75],[2624.0013,2],[2668.5025,2],[2690.5025,2.5],[2750.5,2.5],[2779,3]]
       }
       Scale.Y {
        type=MovieSceneScriptingDoubleChannel
        default=2
        thinned_from=12
        keys [[1864.5,1.25],[2529.0025,1.25],[2558.5025,1.75],[2598.0013,1.75],[2624.0013,2],[2668.5025,2],[2690.5025,2.5],[2750.5,2.5],[2779,3]]
       }
       Scale.Z {
        type=MovieSceneScriptingDoubleChannel
        default=2
        thinned_from=12
        keys [[1864.5,1.25],[2529.0025,1.25],[2558.5025,1.75],[2598.0013,1.75],[2624.0013,2],[2668.5025,2],[2690.5025,2.5],[2750.5,2.5],[2779,3]]
       }
      ]
     }
     {
      type=MovieScene3DTransformSection
      channels [
       Location.X {type=MovieSceneScriptingDoubleChannel, default=0, const [14163.7002,1]}
       Location.Y {type=MovieSceneScriptingDoubleChannel, default=0, const [76618.9219,1]}
       Location.Z {type=MovieSceneScriptingDoubleChannel, default=0, const [3947.8015,1]}
       Rotation.X {type=MovieSceneScriptingDoubleChannel, default=0, const [0,1]}
       Rotation.Y {type=MovieSceneScriptingDoubleChannel, default=0, const [0,1]}
       Rotation.Z {type=MovieSceneScriptingDoubleChannel, default=0, const [0,1]}
       Scale.X {type=MovieSceneScriptingDoubleChannel, default=0, const [1,1]}
       Scale.Y {type=MovieSceneScriptingDoubleChannel, default=0, const [1,1]}
       Scale.Z {type=MovieSceneScriptingDoubleChannel, default=0, const [1,1]}
      ]
     }
    ]
   }
   {
    type=MovieSceneVisibilityTrack
    sections [{
      type=MovieSceneVisibilitySection
      channels [None {type=MovieSceneScriptingBoolChannel, default=true, keys [[0,true],[1800,false],[1828.5,true],[3955,true],[3955.5,false]]}]
     }]
   }
  ]
 }
 CameraComponent {
  id=#6
  parent_id=#1
  possessed_class=CineCameraComponent
  tracks [
   {
    type=MovieSceneFloatTrack
    display_name=Current Focal Length
    sections [{type=MovieSceneFloatSection, channels [None {type=MovieSceneScriptingFloatChannel, default=24}]}]
   }
   {
    type=MovieSceneFloatTrack
    display_name="Manual Focus Distance (Focus Settings)"
    sections [{type=MovieSceneFloatSection, channels [None {type=MovieSceneScriptingFloatChannel, default=100000}]}]
   }
   {
    type=MovieSceneFloatTrack
    display_name=Current Aperture
    sections [{type=MovieSceneFloatSection, channels [None {type=MovieSceneScriptingFloatChannel, default=2.8}]}]
   }
  ]
 }
 Tentacle_PistolCanary_Creature_Idle_01_AngryIdle {
  id=#7
  possessed_class=FortSkeletalMeshActor
  tracks [
   {type=MovieSceneSpawnTrack, sections [{type=MovieSceneSpawnSection, channels [None {type=MovieSceneScriptingBoolChannel, default=true}]}]}
   {type=MovieSceneSkeletalAnimationTrack, sections []}
   {
    type=MovieSceneControlRigParameterTrack
    display_name=Tentacle_PistolCanary_Creature_CtrlRig
    sections [
     {
      type=MovieSceneControlRigParameterSection
      channels [
       control rig options.Location.X {type=MovieSceneScriptingFloatChannel, default=0}
       control rig options.Location.Y {type=MovieSceneScriptingFloatChannel, default=0}
       control rig options.Location.Z {type=MovieSceneScriptingFloatChannel, default=0}
       control rig options.Rotation.X {type=MovieSceneScriptingFloatChannel, default=0}
       control rig options.Rotation.Y {type=MovieSceneScriptingFloatChannel, default=0}
       control rig options.Rotation.Z {type=MovieSceneScriptingFloatChannel, default=0}
       control rig options.Scale.X {type=MovieSceneScriptingFloatChannel, default=1}
       control rig options.Scale.Y {type=MovieSceneScriptingFloatChannel, default=1}
       control rig options.Scale.Z {type=MovieSceneScriptingFloatChannel, default=1}
       Weight {type=MovieSceneScriptingFloatChannel, default=1}
      ]
     }
     {type=MovieSceneControlRigParameterSection, channels [Weight {type=MovieSceneScriptingFloatChannel, default=1}]}
    ]
   }
   {
    type=MovieScene3DTransformTrack
    sections [{
      type=MovieScene3DTransformSection
      channels [
       Location.X {type=MovieSceneScriptingDoubleChannel, default=14287.208, const [14287.208,1]}
       Location.Y {type=MovieSceneScriptingDoubleChannel, default=70618.9219, const [70618.9219,1]}
       Location.Z {type=MovieSceneScriptingDoubleChannel, default=3947.8015, const [3947.8015,1]}
       Rotation.X {type=MovieSceneScriptingDoubleChannel, default=0}
       Rotation.Y {type=MovieSceneScriptingDoubleChannel, default=0}
       Rotation.Z {type=MovieSceneScriptingDoubleChannel, default=0}
       Scale.X {type=MovieSceneScriptingDoubleChannel, default=1}
       Scale.Y {type=MovieSceneScriptingDoubleChannel, default=1}
       Scale.Z {type=MovieSceneScriptingDoubleChannel, default=1}
      ]
     }]
   }
  ]
 }
 Tentacle_PistolCanary_Creature_Layout_02 {
  id=#8
  possessed_class=FortSkeletalMeshActor
  tracks [
   {
    type=MovieSceneSkeletalAnimationTrack
    sections [
     {type=MovieSceneSkeletalAnimationSection, end_frame=3693, start_frame=3683}
     {type=MovieSceneSkeletalAnimationSection, end_frame=3810, start_frame=3693}
     {type=MovieSceneSkeletalAnimationSection, end_frame=3872, start_frame=3765}
     {type=MovieSceneSkeletalAnimationSection, end_frame=4004, start_frame=3828}
     {type=MovieSceneSkeletalAnimationSection, end_frame=4136, start_frame=3954}
    ]
   }
   {type=MovieSceneSpawnTrack, sections [{type=MovieSceneSpawnSection, channels [None {type=MovieSceneScriptingBoolChannel, default=true}]}]}
   {
    type=MovieScene3DTransformTrack
    sections [
     {
      type=MovieScene3DTransformSection
      channels [
       Location.X {type=MovieSceneScriptingDoubleChannel, default=0, keys [[3688,123.5116],[3830,123.5116],[3907,900],[3955,0]]}
       Location.Y {type=MovieSceneScriptingDoubleChannel, default=-6000, keys [[3688,-6000],[3830,-6000],[3907,-8000],[3955,-6000]]}
       Location.Z {
        type=MovieSceneScriptingDoubleChannel
        default=-800
        keys [[3688,-18000],[3690.5,0],[3735.5,0],[3830,-600],[3907,-1000],[3910.5,-800],[3955,-600]]
       }
       Rotation.X {type=MovieSceneScriptingDoubleChannel, default=0, const [0,3]}
       Rotation.Y {type=MovieSceneScriptingDoubleChannel, default=-0.1356, const [-0.1356,3]}
       Rotation.Z {type=MovieSceneScriptingDoubleChannel, default=0, keys [[3688,0],[3830,0],[3907,-10],[3955,0]]}
       Scale.X {type=MovieSceneScriptingDoubleChannel, default=0.75, const [0.75,1]}
       Scale.Y {type=MovieSceneScriptingDoubleChannel, default=0.75, const [0.75,1]}
       Scale.Z {type=MovieSceneScriptingDoubleChannel, default=0.75, const [0.75,1]}
      ]
     }
     {
      type=MovieScene3DTransformSection
      channels [
       Location.X {type=MovieSceneScriptingDoubleChannel, default=14436, const [14436,1]}
       Location.Y {type=MovieSceneScriptingDoubleChannel, default=76088, const [76088,1]}
       Location.Z {type=MovieSceneScriptingDoubleChannel, default=3072, const [3072,1]}
       Rotation.X {type=MovieSceneScriptingDoubleChannel, default=0, const [0,1]}
       Rotation.Y {type=MovieSceneScriptingDoubleChannel, default=0, const [0,1]}
       Rotation.Z {type=MovieSceneScriptingDoubleChannel, default=0, const [0,1]}
       Scale.X {type=MovieSceneScriptingDoubleChannel, default=0, const [0,1]}
       Scale.Y {type=MovieSceneScriptingDoubleChannel, default=0, const [0,1]}
       Scale.Z {type=MovieSceneScriptingDoubleChannel, default=0, const [0,1]}
      ]
     }
    ]
   }
   {
    type=MovieSceneControlRigParameterTrack
    display_name=Tentacle_PistolCanary_Creature_CtrlRig
    sections [{
      type=MovieSceneControlRigParameterSection
      channels [
       control rig options.Location.X {type=MovieSceneScriptingFloatChannel, default=0}
       control rig options.Location.Y {type=MovieSceneScriptingFloatChannel, default=0}
       control rig options.Location.Z {type=MovieSceneScriptingFloatChannel, default=0}
       control rig options.Rotation.X {type=MovieSceneScriptingFloatChannel, default=0}
       control rig options.Rotation.Y {type=MovieSceneScriptingFloatChannel, default=0}
       control rig options.Rotation.Z {type=MovieSceneScriptingFloatChannel, default=0}
       control rig options.Scale.X {type=MovieSceneScriptingFloatChannel, default=1}
       control rig options.Scale.Y {type=MovieSceneScriptingFloatChannel, default=1}
       control rig options.Scale.Z {type=MovieSceneScriptingFloatChannel, default=1}
      ]
     }]
   }
  ]
 }
 Tentacle Pistol Canary Creature - Layout {
  id=#9
  possessed_class=SkeletalMeshActor
  tracks [
   {type=MovieSceneSpawnTrack, sections [{type=MovieSceneSpawnSection, channels [None {type=MovieSceneScriptingBoolChannel, default=true}]}]}
   {
    type=MovieScene3DTransformTrack
    sections [{
      type=MovieScene3DTransformSection
      channels [
       Location.X {type=MovieSceneScriptingDoubleChannel, default=36549.9999, const [7287.21,1]}
       Location.Y {type=MovieSceneScriptingDoubleChannel, default=0, const [70618.9219,1]}
       Location.Z {type=MovieSceneScriptingDoubleChannel, default=0, const [3947.8015,1]}
       Rotation.X {type=MovieSceneScriptingDoubleChannel, default=0, const [0,1]}
       Rotation.Y {type=MovieSceneScriptingDoubleChannel, default=0, const [0,1]}
       Rotation.Z {type=MovieSceneScriptingDoubleChannel, default=0, const [0,1]}
       Scale.X {type=MovieSceneScriptingDoubleChannel, default=1, const [1,1]}
       Scale.Y {type=MovieSceneScriptingDoubleChannel, default=1, const [1,1]}
       Scale.Z {type=MovieSceneScriptingDoubleChannel, default=1, const [1,1]}
      ]
     }]
   }
   {
    type=MovieSceneSkeletalAnimationTrack
    sections [
     {type=MovieSceneSkeletalAnimationSection, end_frame=1, start_frame=-1}
     {type=MovieSceneSkeletalAnimationSection, end_frame=3, start_frame=1}
     {type=MovieSceneSkeletalAnimationSection, end_frame=5, start_frame=3}
     {type=MovieSceneSkeletalAnimationSection, end_frame=7, start_frame=5}
     {type=MovieSceneSkeletalAnimationSection, end_frame=92, start_frame=7}
     {type=MovieSceneSkeletalAnimationSection, end_frame=248, start_frame=92}
     {type=MovieSceneSkeletalAnimationSection, end_frame=403, start_frame=248}
    ]
   }
   {
    type=MovieSceneControlRigParameterTrack
    display_name=Tentacle_PistolCanary_Creature_CtrlRig
    sections [{
      type=MovieSceneControlRigParameterSection
      channels [
       control rig options.Location.X {type=MovieSceneScriptingFloatChannel, default=0}
       control rig options.Location.Y {type=MovieSceneScriptingFloatChannel, default=0}
       control rig options.Location.Z {type=MovieSceneScriptingFloatChannel, default=0}
       control rig options.Rotation.X {type=MovieSceneScriptingFloatChannel, default=0}
       control rig options.Rotation.Y {type=MovieSceneScriptingFloatChannel, default=0}
       control rig options.Rotation.Z {type=MovieSceneScriptingFloatChannel, default=0}
       control rig options.Scale.X {type=MovieSceneScriptingFloatChannel, default=1}
       control rig options.Scale.Y {type=MovieSceneScriptingFloatChannel, default=1}
       control rig options.Scale.Z {type=MovieSceneScriptingFloatChannel, default=1}
      ]
     }]
   }
  ]
 }
 Tentacle_PistolCanary_Creature_Idle_World {
  id=#10
  possessed_class=FortSkeletalMeshActor
  tracks [
   {type=MovieSceneSpawnTrack, sections [{type=MovieSceneSpawnSection, channels [None {type=MovieSceneScriptingBoolChannel, default=true}]}]}
   {type=MovieSceneSkeletalAnimationTrack, sections []}
   {
    type=MovieScene3DTransformTrack
    sections [{
      type=MovieScene3DTransformSection
      channels [
       Location.X {type=MovieSceneScriptingDoubleChannel, default=14287.208, const [14287.208,1]}
       Location.Y {type=MovieSceneScriptingDoubleChannel, default=70618.9219, const [70618.9219,1]}
       Location.Z {type=MovieSceneScriptingDoubleChannel, default=3947.8015, const [3947.8015,1]}
       Rotation.X {type=MovieSceneScriptingDoubleChannel, default=0}
       Rotation.Y {type=MovieSceneScriptingDoubleChannel, default=0}
       Rotation.Z {type=MovieSceneScriptingDoubleChannel, default=0}
       Scale.X {type=MovieSceneScriptingDoubleChannel, default=1}
       Scale.Y {type=MovieSceneScriptingDoubleChannel, default=1}
       Scale.Z {type=MovieSceneScriptingDoubleChannel, default=1}
      ]
     }]
   }
   {
    type=MovieSceneControlRigParameterTrack
    display_name=Tentacle_PistolCanary_Creature_CtrlRig
    sections [{
      type=MovieSceneControlRigParameterSection
      channels [
       control rig options.Location.X {type=MovieSceneScriptingFloatChannel, default=0}
       control rig options.Location.Y {type=MovieSceneScriptingFloatChannel, default=0}
       control rig options.Location.Z {type=MovieSceneScriptingFloatChannel, default=0}
       control rig options.Rotation.X {type=MovieSceneScriptingFloatChannel, default=0}
       control rig options.Rotation.Y {type=MovieSceneScriptingFloatChannel, default=0}
       control rig options.Rotation.Z {type=MovieSceneScriptingFloatChannel, default=0}
       control rig options.Scale.X {type=MovieSceneScriptingFloatChannel, default=1}
       control rig options.Scale.Y {type=MovieSceneScriptingFloatChannel, default=1}
       control rig options.Scale.Z {type=MovieSceneScriptingFloatChannel, default=1}
       Weight {type=MovieSceneScriptingFloatChannel, default=1}
      ]
     }]
   }
  ]
 }
 Tentacle_PistolCanary_Creature_WIP {
  id=#11
  possessed_class=FortSkeletalMeshActor
  tracks [
   {type=MovieSceneSpawnTrack, sections [{type=MovieSceneSpawnSection, channels [None {type=MovieSceneScriptingBoolChannel, default=true}]}]}
   {type=MovieSceneSkeletalAnimationTrack, sections []}
   {
    type=MovieSceneControlRigParameterTrack
    display_name=Tentacle_PistolCanary_Creature_CtrlRig
    sections [
     {
      type=MovieSceneControlRigParameterSection
      channels [
       control rig options.Location.X {type=MovieSceneScriptingFloatChannel, default=0}
       control rig options.Location.Y {type=MovieSceneScriptingFloatChannel, default=0}
       control rig options.Location.Z {type=MovieSceneScriptingFloatChannel, default=0}
       control rig options.Rotation.X {type=MovieSceneScriptingFloatChannel, default=0}
       control rig options.Rotation.Y {type=MovieSceneScriptingFloatChannel, default=0}
       control rig options.Rotation.Z {type=MovieSceneScriptingFloatChannel, default=0}
       control rig options.Scale.X {type=MovieSceneScriptingFloatChannel, default=1}
       control rig options.Scale.Y {type=MovieSceneScriptingFloatChannel, default=1}
       control rig options.Scale.Z {type=MovieSceneScriptingFloatChannel, default=1}
       Weight {type=MovieSceneScriptingFloatChannel, default=1}
      ]
     }
     {type=MovieSceneControlRigParameterSection, channels [Weight {type=MovieSceneScriptingFloatChannel, default=1}]}
    ]
   }
   {
    type=MovieScene3DTransformTrack
    sections [{
      type=MovieScene3DTransformSection
      channels [
       Location.X {type=MovieSceneScriptingDoubleChannel, default=36549.9999, const [14287.208,1]}
       Location.Y {type=MovieSceneScriptingDoubleChannel, default=0, const [70618.9219,1]}
       Location.Z {type=MovieSceneScriptingDoubleChannel, default=0, const [3947.8015,1]}
       Rotation.X {type=MovieSceneScriptingDoubleChannel, default=0, const [0,1]}
       Rotation.Y {type=MovieSceneScriptingDoubleChannel, default=0, const [0,1]}
       Rotation.Z {type=MovieSceneScriptingDoubleChannel, default=0, const [0,1]}
       Scale.X {type=MovieSceneScriptingDoubleChannel, default=1, const [1,1]}
       Scale.Y {type=MovieSceneScriptingDoubleChannel, default=1, const [1,1]}
       Scale.Z {type=MovieSceneScriptingDoubleChannel, default=1, const [1,1]}
      ]
     }]
   }
  ]
 }
]
display_rate {denominator=1, numerator=30}
tick_resolution {denominator=1, numerator=24000}
tracks [{type=MovieSceneCameraCutTrack, sections [{type=MovieSceneCameraCutSection, end_frame=846, start_frame=0}]}]
