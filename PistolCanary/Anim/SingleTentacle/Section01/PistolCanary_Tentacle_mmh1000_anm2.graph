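# Level Sequence for PistolCanary's Section01 single-tentacle animation, staging a spawned tentacle creature and a MaskMaker character on a 30 fps timeline.
# The tentacle holds a fixed transform, with a Control Rig parameter track that shows only unkeyed option values.
# MaskMaker chains overlapping skeletal animations while keyed transforms move, turn and enlarge it in steps.
# A second, constant transform section carries a large location offset, and a visibility track limits when MaskMaker appears.
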LevelSequence /PistolCanary/Anim/SingleTentacle/Section01/PistolCanary_Tentacle_mmh1000_anm2
bindings [
 Tentacle_PistolCanary_Creature {
  id=#1
  possessed_class=FortSkeletalMeshActor
  tracks [
   {type=MovieSceneSpawnTrack, sections [{type=MovieSceneSpawnSection, channels [None {type=MovieSceneScriptingBoolChannel, default=true}]}]}
   {
    type=MovieScene3DTransformTrack
    sections [{
      type=MovieScene3DTransformSection
      channels [
       Location.X {type=MovieSceneScriptingDoubleChannel, default=36549.9999, const [14287.208,1]}
       Location.Y {type=MovieSceneScriptingDoubleChannel, default=0, const [70618.9219,1]}
       Location.Z {type=MovieSceneScriptingDoubleChannel, default=0, const [3947.8015,1]}
       Rotation.X {type=MovieSceneScriptingDoubleChannel, default=0, const [0,1]}
       Rotation.Y {type=MovieSceneScriptingDoubleChannel, default=0, const [0,1]}
       Rotation.Z {type=MovieSceneScriptingDoubleChannel, default=0, const [0,1]}
       Scale.X {type=MovieSceneScriptingDoubleChannel, default=1, const [1,1]}
       Scale.Y {type=MovieSceneScriptingDoubleChannel, default=1, const [1,1]}
       Scale.Z {type=MovieSceneScriptingDoubleChannel, default=1, const [1,1]}
      ]
     }]
   }
   {
    type=MovieSceneControlRigParameterTrack
    display_name=Tentacle_PistolCanary_Creature_CtrlRig
    sections [{
      type=MovieSceneControlRigParameterSection
      channels [
       control rig options.Location.X {type=MovieSceneScriptingFloatChannel, default=0}
       control rig options.Location.Y {type=MovieSceneScriptingFloatChannel, default=0}
       control rig options.Location.Z {type=MovieSceneScriptingFloatChannel, default=0}
       control rig options.Rotation.X {type=MovieSceneScriptingFloatChannel, default=-2.2733}
       control rig options.Rotation.Y {type=MovieSceneScriptingFloatChannel, default=0.2639}
       control rig options.Rotation.Z {type=MovieSceneScriptingFloatChannel, default=0.0427}
       control rig options.Scale.X {type=MovieSceneScriptingFloatChannel, default=1}
       control rig options.Scale.Y {type=MovieSceneScriptingFloatChannel, default=1}
       control rig options.Scale.Z {type=MovieSceneScriptingFloatChannel, default=1}
       Weight {type=MovieSceneScriptingFloatChannel, default=1}
      ]
     }]
   }
  ]
 }
 MaskMaker {
  id=#2
  possessed_class=FortPlayerMannequin
  tracks [
   {
    type=MovieSceneSkeletalAnimationTrack
    sections [
     {type=MovieSceneSkeletalAnimationSection, end_frame=2496, start_frame=1800}
     {type=MovieSceneSkeletalAnimationSection, end_frame=2582, start_frame=2475}
     {type=MovieSceneSkeletalAnimationSection, end_frame=2644, start_frame=2576}
     {type=MovieSceneSkeletalAnimationSection, end_frame=2663, start_frame=2634}
     {type=MovieSceneSkeletalAnimationSection, end_frame=2719, start_frame=2657}
     {type=MovieSceneSkeletalAnimationSection, end_frame=2743, start_frame=2714}
     {type=MovieSceneSkeletalAnimationSection, end_frame=2798, start_frame=2737}
     {type=MovieSceneSkeletalAnimationSection, end_frame=3076, start_frame=2779}
     {type=MovieSceneSkeletalAnimationSection, end_frame=3095, start_frame=3075}
     {type=MovieSceneSkeletalAnimationSection, end_frame=3200, start_frame=3091}
     {type=MovieSceneSkeletalAnimationSection, end_frame=3977, start_frame=3182}
    ]
   }
   {type=MovieSceneSpawnTrack, sections [{type=MovieSceneSpawnSection, channels [None {type=MovieSceneScriptingBoolChannel, default=true}]}]}
   {
    type=MovieScene3DTransformTrack
    sections [
     {
      type=MovieScene3DTransformSection
      channels [
       Location.X {type=MovieSceneScriptingDoubleChannel, default=-1950, keys [[1864.5,-100],[3895,-100],[3908,-1250],[3909,-1950],[3954.5,-46555]]}
       Location.Y {
        type=MovieSceneScriptingDoubleChannel
        default=500
        thinned_from=15
        keys [[1881.5,-6320.5469],[1912.5,-6216.229],[2088.5,-3553.2683],[2344,-375.6404],[3086,-375.6404],[3099,-100],[3173.5,-100],[3193.5,-375.6404],[3329.5,115.8782],[3688,115.8782],[3797,-200],[3883.5,-500],[3907,-500],[3954.5,8164]]
       }
       Location.Z {
        type=MovieSceneScriptingDoubleChannel
        default=1100
        thinned_from=15
        keys [[1864.5,1100],[2161.5,1324],[2368.5,2748],[3193.5,2748],[3329.5,3654.0254],[3419,3877.6357],[3735.5,3877.6357],[3797,3800],[3909,3800],[3927.5,13676],[3941,15257],[3954.5,13706]]
       }
       Rotation.X {
        type=MovieSceneScriptingDoubleChannel
        default=20
        thinned_from=18
        keys [[1864.5,0],[2856.5,0],[3075.5,20],[3085.5,46],[3200.5,46],[3257.5,19],[3688.5,19],[3710.5,-5],[3735.5,-10],[3895,-10],[3907,20],[3919.495,120],[3954.5,120]]
       }
       Rotation.Y {type=MovieSceneScriptingDoubleChannel, default=0, const [0,16]}
       Rotation.Z {
        type=MovieSceneScriptingDoubleChannel
        default=100
        thinned_from=15
        keys [[1864.5,0],[2257,0],[2337,180],[2999.0038,200],[3899,200],[3907,100],[3919.495,41],[3954.5,41]]
       }
       Scale.X {
        type=MovieSceneScriptingDoubleChannel
        default=2
        thinned_from=12
        keys [[1864.5,1.25],[2529.0025,1.25],[2558.5025,1.75],[2598.0013,1.75],[2624.0013,2],[2668.5025,2],[2690.5025,2.5],[2750.5,2.5],[2779,3]]
       }
       Scale.Y {
        type=MovieSceneScriptingDoubleChannel
        default=2
        thinned_from=12
        keys [[1864.5,1.25],[2529.0025,1.25],[2558.5025,1.75],[2598.0013,1.75],[2624.0013,2],[2668.5025,2],[2690.5025,2.5],[2750.5,2.5],[2779,3]]
       }
       Scale.Z {
        type=MovieSceneScriptingDoubleChannel
        default=2
        thinned_from=12
        keys [[1864.5,1.25],[2529.0025,1.25],[2558.5025,1.75],[2598.0013,1.75],[2624.0013,2],[2668.5025,2],[2690.5025,2.5],[2750.5,2.5],[2779,3]]
       }
      ]
     }
     {
      type=MovieScene3DTransformSection
      channels [
       Location.X {type=MovieSceneScriptingDoubleChannel, default=14436, const [14436,1]}
       Location.Y {type=MovieSceneScriptingDoubleChannel, default=76088, const [76088,1]}
       Location.Z {type=MovieSceneScriptingDoubleChannel, default=3072, const [3072,1]}
       Rotation.X {type=MovieSceneScriptingDoubleChannel, default=0, const [0,1]}
       Rotation.Y {type=MovieSceneScriptingDoubleChannel, default=0, const [0,1]}
       Rotation.Z {type=MovieSceneScriptingDoubleChannel, default=0, const [0,1]}
       Scale.X {type=MovieSceneScriptingDoubleChannel, default=0, const [0,1]}
       Scale.Y {type=MovieSceneScriptingDoubleChannel, default=0, const [0,1]}
       Scale.Z {type=MovieSceneScriptingDoubleChannel, default=0, const [0,1]}
      ]
     }
    ]
   }
   {
    type=MovieSceneVisibilityTrack
    sections [{
      type=MovieSceneVisibilitySection
      channels [None {type=MovieSceneScriptingBoolChannel, default=false, keys [[1800,false],[1828.5,true],[3955,true],[3955.5,false]]}]
     }]
   }
  ]
 }
]
display_rate {denominator=1, numerator=30}
tick_resolution {denominator=1, numerator=24000}
tracks []
